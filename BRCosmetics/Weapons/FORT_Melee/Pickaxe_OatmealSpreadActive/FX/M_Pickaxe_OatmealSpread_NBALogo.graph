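# Unlit translucent Niagara sprite material that shows an NBA logo texture for the OatmealSpread pickaxe effects.
# The texture parameter's RGB is multiplied by particle color and fed through the shared FX emissive adjustments function.
# Texture alpha times particle alpha forms the opacity value, with a quality switch adding depth fade at default, High and Epic while Low and Medium skip it.
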Material /BRCosmetics/Weapons/FORT_Melee/Pickaxe_OatmealSpreadActive/FX/M_Pickaxe_OatmealSpread_NBALogo
graph {
 properties {
  bAutomaticallySetUsageInEditor=false
  bCastRayTracedShadows=false
  bUsedWithNiagaraSprites=true
  blendMode=BLEND_Translucent
  shadingModel=MSM_Unlit
  usedShadingModels=Unlit
 }
 nodes [
  n1 = TextureSampleParameter2D() {parameterName=Tex, texture=&/T_OatmealSpread_NBALogo}
  n2 = ParticleColor()
  MaterialFunctionCall(Emissive=Multiply(A=@n1.RGB, B=@n2.RGB)) {materialFunction=&MF_FX_EmissiveAdjustments}
  n3 = Multiply(A=@n1.A, B=@n2.A)
  n4 = DepthFade(Opacity=@n3) {fadeDistanceDefault=10}
  QualitySwitch(Default=@n4, Low=@n3, High=@n4, Medium=@n3, Epic=@n4)
 ]
}
