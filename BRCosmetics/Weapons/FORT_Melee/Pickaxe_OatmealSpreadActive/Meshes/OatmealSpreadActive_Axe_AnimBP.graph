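# Pickaxe_OatmealSpreadActive axe AnimBP that adds secondary physics motion to its ring and lace bones with AnimDynamics.
# In the front end, it picks one of two bone-offset-plus-dynamics chains by whether the pickaxe is held, inferred from body mesh validity.
# Outside it, dynamics layers fade by swing state, read from the owner's PlayerMeleeCombo montage.
# Two other AnimDynamics chains are left unconnected to the output.
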AnimBlueprint /BRCosmetics/Weapons/FORT_Melee/Pickaxe_OatmealSpreadActive/Meshes/OatmealSpreadActive_Axe_AnimBP
target_skeleton=&/OatmealSpreadActive_Axe_Skeleton
graphs [
 AnimGraph {nodes [
   n1 = Anim Dynamics - Bone: dyn_ring_01_base!(ComponentPose, bAlphaBoolEnabled:bool=True, GravityOverride:Vector="0,0,0", ExternalForce="0,0,0") -> (Pose) {
    alphaInputType=Bool
    angularDampingOverride=1
    angularSpringConstant=10000
    bUseGravityOverride=true
    bUsePlanarLimit=false
    linearDampingOverride=0.9999
    alphaBoolBlend {blendInTime=0.1, blendOption=Linear, blendOutTime=0.1}
    boundBone {boneName=dyn_ring_01_base}
    componentAppliedLinearAccClamp=<700 700 700>
    componentLinearAccScale=<0.4 0.4 0.4>
    componentLinearVelScale=<0.6 0.6 0.6>
    gravityOverride {z=-980}
    physicsBodyDefinitions [{
      boundBone {boneName=dyn_ring_01_base}
      boxExtents=<8 8 8>
      constraintSetup {angularLimitsMax=<0 0 45>, angularLimitsMin=<0 0 -45>, angularTarget=<0 0 0>}
      localJointOffset=<1 0 0>
     }]
   }
   n2 = Anim Dynamics - Bone: dyn_ring_01!(ComponentPose=@n1, bAlphaBoolEnabled:bool=True, GravityOverride:Vector="0,0,0", ExternalForce="0,0,0") -> (Pose) {
    alphaInputType=Bool
    angularDampingOverride=1
    angularSpringConstant=10000
    bUseGravityOverride=true
    bUsePlanarLimit=false
    linearDampingOverride=0.9999
    alphaBoolBlend {blendInTime=0.1, blendOption=Linear, blendOutTime=0.1}
    boundBone {boneName=dyn_ring_01}
    componentAppliedLinearAccClamp=<700 700 700>
    componentLinearAccScale=<0.4 0.4 0.4>
    componentLinearVelScale=<0.6 0.6 0.6>
    gravityOverride {z=-980}
    physicsBodyDefinitions [{
      boundBone {boneName=dyn_ring_01}
      boxExtents=<8 8 8>
      constraintSetup {angularLimitsMax=<30 0 0>, angularLimitsMin=<-30 0 0>, angularTarget=<0 0 0>}
      localJointOffset=<1 0 0>
     }]
   }
   Anim Dynamics - Bone: dyn_lace!(ComponentPose=@n2, bAlphaBoolEnabled:bool=True, GravityOverride:Vector="0,0,0", ExternalForce="0,0,0") -> (Pose) {
    alphaInputType=Bool
    angularDampingOverride=1
    angularSpringConstant=1000
    bAngularSpring=true
    bOverrideAngularDamping=true
    bUseGravityOverride=true
    bUsePlanarLimit=false
    linearDampingOverride=0.9999
    alphaBoolBlend {blendInTime=0.1, blendOption=Linear, blendOutTime=0.1}
    boundBone {boneName=dyn_lace}
    componentAppliedLinearAccClamp=<700 700 700>
    componentLinearAccScale=<0.4 0.4 0.4>
    componentLinearVelScale=<0.6 0.6 0.6>
    gravityOverride {z=-980}
    physicsBodyDefinitions [{
      boundBone {boneName=dyn_lace}
      boxExtents=<8 8 8>
      constraintSetup {angularLimitsMax=<0 0 15>, angularLimitsMin=<0 0 -30>, angularTarget=<0 0 0>}
      localJointOffset=<1 0 0>
     }]
   }
   n3 = Anim Dynamics - Bone: dyn_ring_01_base!(ComponentPose, Alpha:float=1, ExternalForce="0,0,0") -> (Pose) {
    alphaBoolBlend {blendOption=Linear}
    boundBone {boneName=dyn_ring_01_base}
    physicsBodyDefinitions [{
      boundBone {boneName=dyn_ring_01_base}
      boxExtents=<8 8 8>
      constraintSetup {angularLimitsMax=<0 0 30>, angularLimitsMin=<0 0 -30>, angularTarget=<0 0 0>}
      localJointOffset=<-1 0 0>
     }]
   }
   n4 = Anim Dynamics - Bone: dyn_ring_01!(ComponentPose=@n3, Alpha:float=1, ExternalForce="0,0,0") -> (Pose) {
    alphaBoolBlend {blendOption=Linear}
    boundBone {boneName=dyn_ring_01}
    physicsBodyDefinitions [{
      boundBone {boneName=dyn_ring_01}
      boxExtents=<8 8 8>
      constraintSetup {angularLimitsMax=<30 0 0>, angularLimitsMin=<-30 0 0>, angularTarget=<0 0 0>}
      localJointOffset=<-1 0 0>
     }]
   }
   Anim Dynamics - Bone: dyn_lace!(ComponentPose=@n4, Alpha:float=1, ExternalForce="0,0,0") -> (Pose) {
    alphaBoolBlend {blendOption=Linear}
    boundBone {boneName=dyn_lace}
    physicsBodyDefinitions [{
      boundBone {boneName=dyn_lace}
      boxExtents=<8 8 8>
      constraintSetup {angularLimitsMax=<0 60 0>, angularLimitsMin=<0 -15 0>, angularTarget=<0 0 0>}
      localJointOffset=<-1 0 0>
     }]
   }
   n5 = "Transform (Modify) Bone - Bone: dyn_ring_01_base"(
    ComponentPose=Local To Component(LocalPose=Use cached pose 'Start pose'())
    Alpha=1
    Translation="-1.442,0.133,1.697"
    Rotation="-0.000,-0.000,-25.000"
    Scale="1,1,1"
   ) {
    rotationMode=BMM_Additive
    rotationSpace=BCS_BoneSpace
    translationMode=BMM_Additive
    translationSpace=BCS_BoneSpace
    alphaBoolBlend {blendOption=Linear}
    boneToModify {boneName=dyn_ring_01_base}
    rotation=<rot 0 -0.0001 -25>
    translation=<-1.442 0.133 1.697>
   }
   n6 = "Transform (Modify) Bone - Bone: dyn_ring_01"(ComponentPose=@n5, Alpha=1, Translation="0,0,0", Rotation="0,0,0", Scale="1,1,1") {
    rotationMode=BMM_Additive
    rotationSpace=BCS_BoneSpace
    alphaBoolBlend {blendOption=Linear}
    boneToModify {boneName=dyn_ring_01}
    rotation=<rot 0 0 0>
   }
   n7 = "Transform (Modify) Bone - Bone: dyn_lace"(ComponentPose=@n6, Alpha=1, Translation="0,0,0", Rotation="-70.000,0.000,0.000", Scale="1,1,1") {
    rotationMode=BMM_Additive
    rotationSpace=BCS_BoneSpace
    alphaBoolBlend {blendOption=Linear}
    boneToModify {boneName=dyn_lace}
    rotation=<rot -70 0 0>
   }
   n8 = Anim Dynamics - Bone: dyn_ring_01(ComponentPose=@n7, Alpha=1, ExternalForce="0,0,0") {
    angularDampingOverride=0.9999
    bOverrideAngularDamping=true
    bUseGravityOverride=true
    simulationSpace=BoneRelative
    alphaBoolBlend {blendOption=Linear}
    boundBone {boneName=dyn_ring_01}
    componentAppliedLinearAccClamp=<800 800 800>
    componentLinearAccScale=<0.3 0.3 0.3>
    componentLinearVelScale=<0.4 0.4 0.4>
    gravityOverride {z=-980}
    physicsBodyDefinitions [{
      boundBone {boneName=dyn_ring_01}
      boxExtents=<5 5 5>
      constraintSetup {angularLimitsMax=<0 20 0>, angularLimitsMin=<0 -50 0>, angularTarget=<0 0 0>}
      localJointOffset=<-0.5 0 0>
     }]
    relativeSpaceBone {boneName=b_AR_Root}
   }
   n9 = Anim Dynamics - Bone: dyn_lace(ComponentPose=@n8, Alpha=1, ExternalForce="0,0,0") {
    angularDampingOverride=1
    angularSpringConstant=800
    bAngularSpring=true
    bOverrideAngularDamping=true
    bUseGravityOverride=true
    alphaBoolBlend {blendOption=Linear}
    boundBone {boneName=dyn_lace}
    componentAppliedLinearAccClamp=<1000 1000 1000>
    componentLinearAccScale=<0.3 0.3 0.3>
    componentLinearVelScale=<0.4 0.4 0.4>
    gravityOverride {z=-980}
    physicsBodyDefinitions [{
      boundBone {boneName=dyn_lace}
      boxExtents=<10 6 5>
      constraintSetup {angularLimitsMax=<0 5 0>, angularLimitsMin=<0 -30 0>, angularTarget=<0 0 0>}
      localJointOffset=<0.2 0 0>
     }]
    relativeSpaceBone {boneName=dyn_ring_01_base}
   }
   n10 = Component To Local(ComponentPose=@n9)
   n11 = "Transform (Modify) Bone - Bone: dyn_ring_01_base"(
    ComponentPose=Local To Component(LocalPose=Use cached pose 'Start pose'())
    Alpha=1
    Translation="-2.816,1.067,1.539"
    Rotation="4.531,-2.118,-25.084"
    Scale="1,1,1"
   ) {
    rotationMode=BMM_Additive
    rotationSpace=BCS_BoneSpace
    translationMode=BMM_Additive
    translationSpace=BCS_BoneSpace
    alphaBoolBlend {blendOption=Linear}
    boneToModify {boneName=dyn_ring_01_base}
    rotation=<rot 4.531 -2.1181 -25.084>
    translation=<-2.816 1.067 1.539>
   }
   n12 = "Transform (Modify) Bone - Bone: dyn_ring_01"(ComponentPose=@n11, Alpha=1, Translation="0,0,0", Rotation="1.916,24.463,38.129", Scale="1,1,1") {
    rotationMode=BMM_Additive
    rotationSpace=BCS_BoneSpace
    alphaBoolBlend {blendOption=Linear}
    boneToModify {boneName=dyn_ring_01}
    rotation=<rot 1.916 24.4631 38.129>
   }
   n13 = "Transform (Modify) Bone - Bone: dyn_lace"(ComponentPose=@n12, Alpha=1, Translation="0,0,0", Rotation="-35.000,0.000,-0.000", Scale="1,1,1") {
    rotationMode=BMM_Additive
    rotationSpace=BCS_BoneSpace
    alphaBoolBlend {blendOption=Linear}
    boneToModify {boneName=dyn_lace}
    rotation=<rot -35 0.0001 0>
   }
   n14 = Anim Dynamics - Bone: dyn_ring_01(ComponentPose=@n13, Alpha=1, ExternalForce="0,0,0") {
    angularDampingOverride=0.9999
    angularSpringConstant=800
    bAngularSpring=true
    bOverrideAngularDamping=true
    bUseGravityOverride=true
    simulationSpace=World
    alphaBoolBlend {blendOption=Linear}
    boundBone {boneName=dyn_ring_01}
    componentAppliedLinearAccClamp=<800 800 800>
    componentLinearAccScale=<0.4 0.4 0.4>
    componentLinearVelScale=<0.5 0.5 0.5>
    gravityOverride {z=-980}
    physicsBodyDefinitions [{
      boundBone {boneName=dyn_ring_01}
      boxExtents=<5 5 5>
      constraintSetup {angularLimitsMax=<5 15 0>, angularLimitsMin=<-10 -5 0>, angularTarget=<0 0 0>}
      localJointOffset=<-0.5 0 0>
     }]
    relativeSpaceBone {boneName=b_AR_Root}
   }
   n15 = Anim Dynamics - Bone: dyn_lace(ComponentPose=@n14, Alpha=1, ExternalForce="0,0,0") {
    angularDampingOverride=0.99
    angularSpringConstant=600
    bAngularSpring=true
    bOverrideAngularDamping=true
    bUseGravityOverride=true
    simulationSpace=RootRelative
    alphaBoolBlend {blendOption=Linear}
    boundBone {boneName=dyn_lace}
    componentAppliedLinearAccClamp=<1200 1200 1200>
    componentLinearAccScale=<0.5 0.5 0.5>
    componentLinearVelScale=<0.6 0.6 0.6>
    gravityOverride {z=-980}
    physicsBodyDefinitions [{
      boundBone {boneName=dyn_lace}
      boxExtents=<20 10 10>
      constraintSetup {angularLimitsMax=<10 0 0>, angularLimitsMin=<-10 -30 0>, angularTarget=<0 0 0>}
      localJointOffset=<0.2 0 0>
     }]
    relativeSpaceBone {boneName=b_AR_Root}
   }
   n16 = Component To Local(ComponentPose=@n15)
   n17 = Get Character Holding Pickaxe() -> (Character Holding Pickaxe:bool)
   n18 = Get Is Not Swinging() -> (Is Not Swinging:bool)
   n19 = Anim Dynamics - Bone: dyn_ring_01_base!(
    ComponentPose=Local To Component(LocalPose=Use cached pose 'Start pose'())
    bAlphaBoolEnabled:bool=@n18
    GravityOverride:Vector="0,0,0"
    ExternalForce="0,0,0"
   ) -> (Pose) {
    alphaInputType=Bool
    angularDampingOverride=0.9999
    bOverrideAngularDamping=true
    bUseGravityOverride=true
    alphaBoolBlend {blendInTime=0.8, blendOption=Linear, blendOutTime=0.1}
    boundBone {boneName=dyn_ring_01_base}
    componentAppliedLinearAccClamp=<3500 3500 3500>
    componentLinearAccScale=<0.4 0.4 0.4>
    componentLinearVelScale=<0.5 0.5 0.5>
    gravityOverride {z=-980}
    physicsBodyDefinitions [{
      boundBone {boneName=dyn_ring_01_base}
      boxExtents=<8 8 8>
      constraintSetup {angularLimitsMax=<0 5 25>, angularLimitsMin=<0 -5 -25>, angularTarget=<0 0 0>, linearAxesMax=<0 0.5 0.2>, linearAxesMin=<0 -0.2 -0.5>}
      localJointOffset=<1 0 0>
     }]
    relativeSpaceBone {boneName=Melee_TwoHanded_Handle}
   }
   n20 = Anim Dynamics - Bone: dyn_ring_01!(ComponentPose=@n19, bAlphaBoolEnabled:bool=@n18, GravityOverride:Vector="0,0,0", ExternalForce="0,0,0") -> (Pose) {
    alphaInputType=Bool
    angularDampingOverride=0.9999
    bOverrideAngularDamping=true
    bUseGravityOverride=true
    alphaBoolBlend {blendInTime=0.8, blendOption=Linear, blendOutTime=0.1}
    boundBone {boneName=dyn_ring_01}
    componentAppliedLinearAccClamp=<3500 3500 3500>
    componentLinearAccScale=<0.4 0.4 0.4>
    componentLinearVelScale=<0.5 0.5 0.5>
    gravityOverride {z=-980}
    physicsBodyDefinitions [{
      boundBone {boneName=dyn_ring_01}
      boxExtents=<8 8 8>
      constraintSetup {angularLimitsMax=<0 35 0>, angularLimitsMin=<0 -10 0>, angularTarget=<0 0 0>}
      localJointOffset=<1 0 0>
     }]
    relativeSpaceBone {boneName=Melee_TwoHanded_Handle}
   }
   n21 = Anim Dynamics - Bone: dyn_lace!(ComponentPose=@n20, bAlphaBoolEnabled:bool=@n18, GravityOverride:Vector="0,0,0", ExternalForce="0,0,0") -> (Pose) {
    alphaInputType=Bool
    angularDampingOverride=0.9999
    bOverrideAngularDamping=true
    bUseGravityOverride=true
    alphaBoolBlend {blendInTime=0.8, blendOption=Linear, blendOutTime=0.1}
    boundBone {boneName=dyn_lace}
    componentAppliedLinearAccClamp=<3500 3500 3500>
    componentLinearAccScale=<0.4 0.4 0.4>
    componentLinearVelScale=<0.5 0.5 0.5>
    gravityOverride {z=-980}
    physicsBodyDefinitions [{
      boundBone {boneName=dyn_lace}
      boxExtents=<10 10 10>
      constraintSetup {angularLimitsMax=<0 150 0>, angularLimitsMin=<0 0 0>, angularTarget=<0 0 0>}
      localJointOffset=<1 0 0>
     }]
    relativeSpaceBone {boneName=Melee_TwoHanded_Handle}
   }
   n22 = Get Is Swinging() -> (Is Swinging:bool)
   n23 = Anim Dynamics - Bone: dyn_ring_01(ComponentPose=@n21, bAlphaBoolEnabled=@n22, GravityOverride="0,0,0", ExternalForce="0,0,0") {
    alphaInputType=Bool
    angularDampingOverride=0.9999
    bOverrideAngularDamping=true
    bUseGravityOverride=true
    alphaBoolBlend {blendInTime=0.1, blendOption=Linear, blendOutTime=0.3}
    boundBone {boneName=dyn_ring_01}
    componentAppliedLinearAccClamp=<3500 3500 3500>
    componentLinearAccScale=<0.4 0.4 0.4>
    componentLinearVelScale=<0.5 0.5 0.5>
    gravityOverride {z=-980}
    physicsBodyDefinitions [{
      boundBone {boneName=dyn_ring_01}
      boxExtents=<8 8 8>
      constraintSetup {angularLimitsMax=<0 90 0>, angularLimitsMin=<0 80 0>, angularTarget=<0 0 0>}
      localJointOffset=<1 0 0>
     }]
    relativeSpaceBone {boneName=Melee_TwoHanded_Handle}
   }
   n24 = Anim Dynamics - Bone: dyn_lace(ComponentPose=@n23, bAlphaBoolEnabled=@n22, GravityOverride="0,0,0", ExternalForce="0,0,0") {
    alphaInputType=Bool
    angularDampingOverride=0.9999
    bOverrideAngularDamping=true
    bUseGravityOverride=true
    alphaBoolBlend {blendInTime=0.1, blendOption=Linear, blendOutTime=0.3}
    boundBone {boneName=dyn_lace}
    componentAppliedLinearAccClamp=<3500 3500 3500>
    componentLinearAccScale=<0.4 0.4 0.4>
    componentLinearVelScale=<0.5 0.5 0.5>
    gravityOverride {z=-980}
    physicsBodyDefinitions [{
      boundBone {boneName=dyn_lace}
      boxExtents=<10 10 10>
      constraintSetup {angularLimitsMax=<0 150 0>, angularLimitsMin=<0 0 0>, angularTarget=<0 0 0>}
      localJointOffset=<1 0 0>
     }]
    relativeSpaceBone {boneName=Melee_TwoHanded_Handle}
   }
   Output Pose(Result=Blend Poses by bool(
     BlendPose_0=Blend Poses by bool(BlendPose_0=@n10, BlendPose_1=@n16, bActiveValue=@n17) {blendType=Linear}
     BlendPose_1=Component To Local(ComponentPose=@n24)
     bActiveValue=Get bIsInFrontEnd()
    ) {blendType=Linear}) {layerGroup=None}
   AnimGraphNode_SubInput_0 = Input Pose()
   Save cached pose 'Start pose'(Pose=@AnimGraphNode_SubInput_0) {cacheName=Start pose}
  ]}
 EventGraph {nodes [
   Event BlueprintUpdateAnimation()
   > n1 = Sequence()
   > n2 = Gravity Override State Function DEPRECATED(CharacterPartAnimInstance=Self-Reference(), GameTimeInSeconds=GetGameTimeInSeconds(), GravityOverrideParams=Get Gravity Override Params())
   > Set Gravity Override(Gravity Override=@n2)
   > Set Is Swinging(Is Swinging:bool=Montage_IsPlaying(
     self=GetOwnerFortPlayerAnimInstance!(self:Custom Character Part Anim Instance Object Reference) -> (ReturnValue:Fort Player Anim Instance Object Reference)
     Montage=/Game/Animation/Game/MainPlayer/Montages/Melee/PlayerMeleeCombo
    )) -> (Output_Get:bool)
   > Set Is Not Swinging(Is Not Swinging:bool=NOT Boolean(A=Get Is Swinging())) -> (Output_Get:bool)
   n3 = Get!(execute:Exec=@n1.then_1) -> (then:Exec, else:Exec, Body Mesh:Skeletal Mesh Component Budgeted Object Reference)
   > Set Character Holding Pickaxe(Character Holding Pickaxe:bool=true) -> (Output_Get:bool)
   Set Character Holding Pickaxe(execute=@n3.else, Character Holding Pickaxe=false)
   n4 = Event OnCharacterCustomizationLoaded()
   > Set Body Mesh(Body Mesh:Skeletal Mesh Component Budgeted Object Reference=Get Array Item with EFort Custom Part Type(Array=@n4.CustomMeshParts)) -> (Output_Get:Skeletal Mesh Component Budgeted Object Reference)
   TryGetPawnOwner()
  ]}
]
variables [
 Body Mesh {type=SkeletalMeshComponentBudgeted}
 Character Holding Pickaxe {type=boolean}
 Gravity Override {type=Vector}
 Gravity Override Params {type=GravityOverrideParamsStruct}
 Is Not Swinging {type=boolean}
 Is Swinging {type=boolean}
]
